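# Sound cue for accepting a quest in an NPC conversation, played on the in-game non-critical UI sound class.
# A mixer layers a radial-menu click, with slight random pitch variation, under the accepted-quest wave at unequal input volumes.
# An outer modulator with fixed pitch wraps the mix, and the cue's overall volume is lowered.
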SoundCue /Game/Athena/Sounds/NPCs/Conversation/Conversation_AcceptedQuest_Cue
root=Modulator_0
nodes [
 {id=Mixer_0, children [Modulator_1,WavePlayer_0], properties {inputVolume [0.8,1]}}
 {id=Modulator_0, children [Mixer_0], properties {pitchMax=1, pitchMin=1}}
 {id=Modulator_1, children [WavePlayer_1], properties {pitchMax=1.02, pitchMin=0.98, volumeMax=1, volumeMin=1}}
 {id=WavePlayer_0, properties {soundWaveAssetPtr=&/Conversation_AcceptedQuest}}
 {id=WavePlayer_1, properties {soundWaveAssetPtr=&/NPC_Radial_Click}}
]
properties {maxDistance=2097152, soundClassObject=&Fort_UI_InGame_NonCritical, volumeMultiplier=0.7}
