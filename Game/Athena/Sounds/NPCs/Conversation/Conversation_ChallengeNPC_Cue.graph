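# Sound cue for challenge NPC conversations, classed as non-critical in-game UI audio.
# A mixer layers an NPC radial click, given slight random pitch variation and a reduced volume, with the challenge conversation sound.
# The conversation sound starts after a short fixed delay, so it follows just behind the click.
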SoundCue /Game/Athena/Sounds/NPCs/Conversation/Conversation_ChallengeNPC_Cue
root=Mixer_0
nodes [
 {id=Delay_0, children [WavePlayer_0], properties {delayMax=0.08, delayMin=0.08}}
 {id=Mixer_0, children [Modulator_1,Delay_0], properties {inputVolume [0.8,1]}}
 {id=Modulator_1, children [WavePlayer_1], properties {pitchMax=1.02, pitchMin=0.98, volumeMax=1, volumeMin=1}}
 {id=WavePlayer_0, properties {soundWaveAssetPtr=&/Conversation_ChallengeNPC}}
 {id=WavePlayer_1, properties {soundWaveAssetPtr=&/NPC_Radial_Click}}
]
properties {maxDistance=2097152, soundClassObject=&Fort_UI_InGame_NonCritical}
