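# Conversation_DismissNPC_Cue is a sound cue that, by its name, plays when an NPC is dismissed through conversation.
# It plays a single wave, reusing the disguise gadget's disappear sound, at reduced volume.
# It borrows the disguise equip attenuation, routes through the gadget items sound class, and overrides concurrency to allow one instance per owner.
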SoundCue /Game/Athena/Sounds/NPCs/Conversation/Conversation_DismissNPC_Cue
root=WavePlayer_2
nodes [{id=WavePlayer_2, properties {soundWaveAssetPtr=/Game/Athena/Sounds/Gadgets/Disguise/Item_Disguise_Disappear_01}}]
properties {
 attenuationSettings=/Game/Athena/Sounds/Gadgets/Disguise/Disguise_Equip_Att
 bOverrideConcurrency=true
 maxDistance=1712
 soundClassObject=&Fort_Gadgets_Items
 volumeMultiplier=0.65
 concurrencyOverrides {bLimitToOwner=true, maxCount=1}
}
