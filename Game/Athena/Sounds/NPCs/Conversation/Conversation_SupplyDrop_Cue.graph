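# Sound cue for the supply drop option in NPC conversations, on the non-critical in-game UI sound class.
# A Mixer layers a slightly pitch-varied radial click, a hire sting, supply drop placement and incoming sounds, and a cash drop, at individual input volumes.
# Short fixed delays stagger the hire and placement layers, while an Enveloper fades the incoming sound to silence.
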SoundCue /Game/Athena/Sounds/NPCs/Conversation/Conversation_SupplyDrop_Cue
root=Modulator_0
nodes [
 {id=Delay_0, children [WavePlayer_0], properties {delayMax=0.1, delayMin=0.1}}
 {id=Delay_1, children [WavePlayer_5], properties {delayMax=0.35, delayMin=0.35}}
 {
  id=Enveloper_0
  children [WavePlayer_3]
  properties {
   pitchCurve {editorCurveData {keys [{time=0}]}}
   volumeCurve {editorCurveData {keys [<at 0 1>,<at 1.0485 1>,<at 1.3213 0.4561>,<at 1.9132 0>]}}
  }
 }
 {id=Mixer_0, children [Modulator_1,Delay_0,Delay_1,Enveloper_0,WavePlayer_2], properties {inputVolume [0.85,0.35,1,1,0.5]}}
 {id=Modulator_0, children [Mixer_0], properties {pitchMax=1, pitchMin=1}}
 {id=Modulator_1, children [WavePlayer_1], properties {pitchMax=1.02, pitchMin=0.98, volumeMax=1, volumeMin=1}}
 {id=WavePlayer_0, properties {soundWaveAssetPtr=&/Conversation_Hire_01}}
 {id=WavePlayer_1, properties {soundWaveAssetPtr=&/NPC_Radial_Click}}
 {id=WavePlayer_2, properties {soundWaveAssetPtr=/Game/Athena/Sounds/Items/Wads/Cash_Drop_High}}
 {id=WavePlayer_3, properties {soundWaveAssetPtr=/Game/Sounds/Fort_Gadgets/Supply_Drop/supply_drop_incoming_3}}
 {id=WavePlayer_5, properties {soundWaveAssetPtr=/Game/Sounds/Fort_Gadgets/Supply_Drop/supply_drop_place}}
]
properties {maxDistance=2097152, soundClassObject=&Fort_UI_InGame_NonCritical, volumeMultiplier=1}
