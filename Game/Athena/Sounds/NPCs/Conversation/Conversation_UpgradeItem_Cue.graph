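# Sound cue for the upgrade-item option in Athena NPC conversations, played as a non-critical in-game UI sound.
# A mixer layers a radial click wave, given slight random pitch variation, with a quieter storage-transfer UI wave that starts after a short fixed delay.
# An outer modulator with fixed pitch wraps the mix at the root.
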SoundCue /Game/Athena/Sounds/NPCs/Conversation/Conversation_UpgradeItem_Cue
root=Modulator_0
nodes [
 {id=Delay_0, children [WavePlayer_1], properties {delayMax=0.1, delayMin=0.1}}
 {id=Mixer_0, children [Modulator_1,Delay_0], properties {inputVolume [1,0.7]}}
 {id=Modulator_0, children [Mixer_0], properties {pitchMax=1, pitchMin=1}}
 {id=Modulator_1, children [WavePlayer_0], properties {pitchMax=1.02, pitchMin=0.98, volumeMax=1, volumeMin=1}}
 {id=WavePlayer_0, properties {soundWaveAssetPtr=&/NPC_Radial_Click}}
 {id=WavePlayer_1, properties {soundWaveAssetPtr=/Game/Sounds/Fort_UI_Sounds/FrontEnd/Fort_UI_Storage_Transfer03}}
]
properties {maxDistance=2097152, soundClassObject=&Fort_UI_InGame_NonCritical, volumeMultiplier=1}
