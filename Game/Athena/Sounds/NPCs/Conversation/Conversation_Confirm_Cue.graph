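# Sound cue for the confirm sound in NPC conversations, played as non-critical in-game UI audio.
# A Mixer layers three waves at different levels: a pitched-down submenu confirm, a radial click weighted loudest, and a dial hover.
# The confirm and click layers each pass through a Modulator for slight random pitch variation, while the hover layer feeds the mixer directly.
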SoundCue /Game/Athena/Sounds/NPCs/Conversation/Conversation_Confirm_Cue
root=Mixer_0
nodes [
 {id=Mixer_0, children [Modulator_1,Modulator_0,WavePlayer_4], properties {inputVolume [0.75,1.35,1]}}
 {id=Modulator_0, children [WavePlayer_0], properties {pitchMax=0.98}}
 {id=Modulator_1, children [WavePlayer_3], properties {pitchMax=1.01, pitchMin=0.99}}
 {id=WavePlayer_0, properties {soundWaveAssetPtr=&/NPC_Radial_Click}}
 {id=WavePlayer_3, properties {soundWaveAssetPtr=&/NPC_Dial_Submenu_PitchDwn_Confirm}}
 {id=WavePlayer_4, properties {soundWaveAssetPtr=&/NPC_Dial_Hover_02}}
]
properties {maxDistance=2097152, soundClassObject=&Fort_UI_InGame_NonCritical, volumeMultiplier=0.6}
