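# Sound cue for the error sound of the NPC conversation radial dial in Athena.
# A single wave player feeds a modulator that adds a very slight random pitch variation on each play.
# It routes to the non-critical in-game UI sound class at reduced volume.
# A concurrency override limits it to one instance, stopping the oldest when retriggered.
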SoundCue /Game/Athena/Sounds/NPCs/Conversation/Conversation_Error_Cue
root=Modulator_1
nodes [
 {id=Modulator_1, children [WavePlayer_1], properties {pitchMax=1.01, pitchMin=0.99}}
 {id=WavePlayer_1, properties {soundWaveAssetPtr=&/NPC_Radial_Dial_Error_01}}
]
properties {
 bOverrideConcurrency=true
 maxDistance=2097152
 soundClassObject=&Fort_UI_InGame_NonCritical
 volumeMultiplier=0.7
 concurrencyOverrides {maxCount=1, resolutionRule=StopOldest}
 concurrencySet [null]
}
