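# Sound cue for the hover feedback in the NPC conversation UI, routed to the non-critical in-game UI sound class at low volume.
# A mixer layers the radial hover wave at full level with a much quieter submenu hover wave pitched down to a fixed value.
# The radial layer passes through a default modulator for slight variation.
# Playback is limited through the shared conversation UI concurrency set.
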SoundCue /Game/Athena/Sounds/NPCs/Conversation/Conversation_Hover_Cue
root=Mixer_0
nodes [
 {id=Mixer_0, children [Modulator_1,Modulator_0], properties {inputVolume [0.1,1]}}
 {id=Modulator_0, children [WavePlayer_0]}
 {id=Modulator_1, children [WavePlayer_2], properties {pitchMax=0.9, pitchMin=0.9}}
 {id=WavePlayer_0, properties {soundWaveAssetPtr=&/NPC_Radial_Hover}}
 {id=WavePlayer_2, properties {soundWaveAssetPtr=&/NPC_Dial_Submenu_PitchUp_Hover}}
]
properties {
 maxDistance=2097152
 soundClassObject=&Fort_UI_InGame_NonCritical
 volumeMultiplier=0.2
 concurrencyOverrides {maxCount=1, resolutionRule=StopOldest}
 concurrencySet [&/Conversation_UI_Conc]
}
